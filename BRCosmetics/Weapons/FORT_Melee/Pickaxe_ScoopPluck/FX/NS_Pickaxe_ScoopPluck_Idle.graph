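# Idle glow Niagara system for the ScoopPluck pickaxe, with a User.Variant float blending its tint from red to gold.
# SpriteGlows and TextureGlow burst-spawn persistent sprites, using spawn-group switches for per-sprite size, socket offset and alignment; TextureGlow also renders a light.
# Alpha curves fade the glows in, scaled for front end versus gameplay and time of day.
# A system Timeline drives brightness flares on TextureGlow.
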NiagaraSystem /BRCosmetics/Weapons/FORT_Melee/Pickaxe_ScoopPluck/FX/NS_Pickaxe_ScoopPluck_Idle
dependencies {
 usedDataInterfaces [&NiagaraDataInterfaceColorCurve, &NiagaraDataInterfaceCurve, &NiagaraDataInterfaceSocketReader]
 usedDynamicInputs [&SetBoolByIntComparison, &MakeCustomFloatFromBool]
 usedModules [
  &self:SystemSpawnScript.SystemScriptSource.SystemScriptGraph.NiagaraNodeAssignment_0.SetVariables_#1
  &self:SystemSpawnScript.SystemScriptSource.SystemScriptGraph.NiagaraNodeAssignment_3.SetVariables_#2
  &SystemState
  &self:SystemSpawnScript.SystemScriptSource.SystemScriptGraph.NiagaraNodeAssignment_1.SetVariables_#3
  &self:SystemSpawnScript.SystemScriptSource.SystemScriptGraph.NiagaraNodeAssignment_4.SetVariables_#4
  &Timeline
  &self:SystemSpawnScript.SystemScriptSource.SystemScriptGraph.NiagaraNodeAssignment_2.SetVariables_#5
 ]
}
emitters {
 SpriteGlows {
  emitterUpdateScript [
   &EmitterState(
    Inactive Response=::NewEnumerator1
    Life Cycle Mode=::NewEnumerator1
    Loop Behavior=::NewEnumerator1
    Loop Duration Mode=::NewEnumerator1
    UseLoopDelay=false
   )
   &SpawnBurst_Instantaneous(Spawn Count=8, Use Spawn Probability=false, UseLoopCountLimit=false)
  ]
  particleSpawnScript [
   &FN_SetSpawnGroup(Number of Spawn Groups=8)
   &InitializeParticle(
    Color=&MergeHelper_ColorSwitch(
     0=<rgba 0.5417 0.5417 0.5417 1>
     1=<rgba 0.5 0.5 0.5 1>
     2=<rgba 0.77 0.77 0.77 1>
     3=<rgba 0.77 0.77 0.77 1>
     4=<rgba 0.77 0.77 0.77 1>
     Index=&SpawnGroup()
     Num=::NewEnumerator3
    )
    Sprite Rotation Mode=::NewEnumerator0
    Sprite Size=&MergeHelper_Vector2Switch(
     0=<11 18.6813>
     1=<11 63>
     2=<11 22.0094>
     3=<11 20.5983>
     4=<11 19.4647>
     5=<0 64.3886>
     6=<0 24.602>
     7=<0 25.2065>
     Index=&SpawnGroup()
     Num=::NewEnumerator6
    )
    Sprite Size Mode=::NewEnumerator4
    UsePositionOffset=false
    Write Material Random=true
   )
   &CameraOffset(Camera Offset Amount=1)
  ]
  particleUpdateScript [
   &ParticleState(Kill Particles When Lifetime Has Elapsed=false, Let Infinitely Lived Particles Die When Emitter Deactivates=true)
   &SocketLocation(
    SamplePosOffset=true
    Sampled Position Offset=&MergeHelper_VectorSwitch(
     0=<0 -14.8961 7.7891>
     1=<0 2.4727 -21.6>
     2=<0 -24.584 -6.3127>
     3=<0 -21.0798 5.5118>
     4=<0 -26.17 0.488>
     5=<0 14.8938 -23.8722>
     6=<0 -13.264 -4.0217>
     7=<0 -11.2139 5.4101>
     Index=&SpawnGroup()
     Num=::NewEnumerator6
    )
    Socket Reader=$System.Socket Reader
    Socket Sampling Mode=::NewEnumerator3
    VelocitySpeedLimit=false
   )
   &ScaleColor(
    Color Scale:LinearColor=&LerpLinearColors(StartColor=<rgba 1.7 0.655 0.655 0.145>, EndColor=<rgba 1 0.864 0.474 0.11>, LerpFactor=$User.Variant)
    Scale Mode=3
   ) {module=ScaleColor001}
   &ScaleColor(
    Scale Alpha=&Multiply_Float(
     A=&FloatFromCurve(FloatCurve=DataInterface() {propertyValues {Curve {
         defaultValue=0
         keys [
          {arriveTangent=4.1785, interpMode=RCIM_Cubic, leaveTangent=4.1785, tangentMode=RCTM_User, time=0.1535, value=0}
          {arriveTangent=1.1179, interpMode=RCIM_Cubic, leaveTangent=1.1179, tangentMode=RCTM_User, time=0.362, value=0.6228}
          {arriveTangent=0.0224, interpMode=RCIM_Cubic, leaveTangent=0.0224, tangentMode=RCTM_User, time=1, value=1}
         ]
        }}})
     B=&Multiply_Float(
      A=&Multiply_Float(
       A=&FN_FrontEndFloatModifier(Gameplay=0.1, Front End=0.05)
       B=&FNTimeOfDayScalar(Night Scale=0.7, Morning Scale=0.85, Evening Scale=0.83)
      )
      B=4
     )
    )
    Scale RGB=&MakeVectorFromLinearColorRGB(Color=<rgba 1 1 1 1>)
    ScaleA=true
    ScaleRGB=true
   )
   &SpriteFacingAndAlignment(
    Sprite Alignment=&MergeHelper_VectorSwitch(
     0=&RotateVectorByQuaternion(Vector To Rotate=<0 1 0.24>, Quaternion=$Output.SocketLocation.SampledRotation)
     1=&RotateVectorByQuaternion(Vector To Rotate=<0 1 -0.456>, Quaternion=$Output.SocketLocation.SampledRotation)
     2=&RotateVectorByQuaternion(Vector To Rotate=<0 -0.7119 1>, Quaternion=$Output.SocketLocation.SampledRotation)
     3=&RotateVectorByQuaternion(Vector To Rotate=<0 1.9088 1>, Quaternion=$Output.SocketLocation.SampledRotation)
     4=&RotateVectorByQuaternion(Vector To Rotate=<0 0.312 1>, Quaternion=$Output.SocketLocation.SampledRotation)
     5=&RotateVectorByQuaternion(Vector To Rotate=<0 -1.7582 1>, Quaternion=$Output.SocketLocation.SampledRotation)
     6=&RotateVectorByQuaternion(Vector To Rotate=<0 -0.656 1>, Quaternion=$Output.SocketLocation.SampledRotation)
     7=&RotateVectorByQuaternion(Vector To Rotate=<0 1.8909 1>, Quaternion=$Output.SocketLocation.SampledRotation)
     Index=&SpawnGroup()
     Num=::NewEnumerator6
    )
    WriteAlignment=true
    WriteFacing=false
   )
  ]
  rendererClasses [&NiagaraSpriteRendererProperties]
  renderers [{rendererClass=&NiagaraSpriteRendererProperties}]
 }
 TextureGlow {
  emitterUpdateScript [
   &EmitterState(
    Inactive Response=::NewEnumerator1
    Life Cycle Mode=::NewEnumerator1
    Loop Behavior=::NewEnumerator1
    Loop Duration Mode=::NewEnumerator1
    UseLoopDelay=false
   )
   &SpawnBurst_Instantaneous(Spawn Count=3, Use Spawn Probability=false, UseLoopCountLimit=false)
  ]
  particleSpawnScript [
   &FN_SetSpawnGroup(Number of Spawn Groups=3)
   &InitializeParticle(
    Lifetime=5
    Sprite Size=&MergeHelper_Vector2Switch(0=<64 82>, 1=<0 0>, 2=<71.5438 84.0536>, 3=?, 4=?, Index=&SpawnGroup(), Num=::NewEnumerator1)
    Sprite Size Mode=::NewEnumerator4
    UsePositionOffset=false
    Write Material Random=true
   )
   &SpriteFacingAndAlignment(
    Sprite Alignment=&MergeHelper_VectorSwitch(0=<9.8248 0 1>, 1=<0 0 0>, 2=<8.1048 0 1>, 3=?, Index=&SpawnGroup(), Num=::NewEnumerator1)
    Sprite Facing=<0 1 0>
    WriteAlignment=true
    WriteFacing=true
   )
   &CameraOffset(Camera Offset Amount=1.5)
   /SetVariables(Particles.VisibilityTag=&SpawnGroup()) {isSetParameters=true}
  ]
  particleUpdateScript [
   &ParticleState(
    Kill Particles When Lifetime Has Elapsed=false
    Let Infinitely Lived Particles Die When Emitter Deactivates=true
    Loop Particles Lifetime=true
   )
   &ScaleColor(
    Color Scale:LinearColor=&MergeHelper_ColorSwitch(
     0=&Multiply_LinearColorByFloat(
      Linear Color=&LerpLinearColors(StartColor=<rgba 3 1.312 1.312 0.497>, EndColor=<rgba 3.5 3.069 1.203 0.34>, LerpFactor=$User.Variant)
      Float=&FN_FrontEndFloatModifier(Gameplay=1, Front End=0.68)
     )
     1=&Multiply_LinearColorByFloat(
      Linear Color=&LerpLinearColors(StartColor=<rgba 3 1.312 1.312 0.497>, EndColor=<rgba 3.5 3.069 1.203 0.34>, LerpFactor=$User.Variant)
      Float=&FN_FrontEndFloatModifier(Gameplay=1, Front End=0.68)
     )
     2=&LerpLinearColors(StartColor=<rgba 2.2 0.149 0.149 0.497>, EndColor=<rgba 2.3 1.4549 0.3594 0.33>, LerpFactor=$User.Variant)
     3=?
     Index=&SpawnGroup()
     Num=::NewEnumerator1
    )
    Scale Mode=3
   ) {module=ScaleColor001}
   &ScaleColor(
    Scale Alpha=&MergeHelper_FloatSwitch(
     0=&FloatFromCurve(
      FloatCurve=DataInterface() {propertyValues {Curve {
         defaultValue=0
         keys [
          {arriveTangent=4.1785, interpMode=RCIM_Cubic, leaveTangent=4.1785, tangentMode=RCTM_User, time=0.1535, value=0}
          {arriveTangent=1.1179, interpMode=RCIM_Cubic, leaveTangent=1.1179, tangentMode=RCTM_User, time=0.362, value=0.6228}
          {arriveTangent=0.0224, interpMode=RCIM_Cubic, leaveTangent=0.0224, tangentMode=RCTM_User, time=1, value=1}
         ]
        }}}
      CurveIndex=$Engine.System.Age
      Scale Curve=&Multiply_Float(
       A=&Multiply_Float(A=&FN_FrontEndFloatModifier(Gameplay=0.2, Front End=0.15), B=$System.TODFloat)
       B=&FloatFromCurve(
        FloatCurve=DataInterface() {propertyValues {Curve {keys [<at 0 1 RCIM_Cubic>,<at 0.15 3 RCIM_Cubic>,<at 1 1>]}}}
        CurveIndex=$System.Timeline.Time
       )
      )
     )
     1=&FloatFromCurve(
      FloatCurve=DataInterface() {propertyValues {Curve {
         defaultValue=0
         keys [
          {arriveTangent=4.1785, interpMode=RCIM_Cubic, leaveTangent=4.1785, tangentMode=RCTM_User, time=0.1535, value=0}
          {arriveTangent=1.1179, interpMode=RCIM_Cubic, leaveTangent=1.1179, tangentMode=RCTM_User, time=0.362, value=0.6228}
          {arriveTangent=0.0224, interpMode=RCIM_Cubic, leaveTangent=0.0224, tangentMode=RCTM_User, time=1, value=1}
         ]
        }}}
      CurveIndex=$Engine.System.Age
      Scale Curve=&Multiply_Float(
       A=&Multiply_Float(A=&FN_FrontEndFloatModifier(Gameplay=0.2, Front End=0.15), B=$System.TODFloat)
       B=&FloatFromCurve(
        FloatCurve=DataInterface() {propertyValues {Curve {keys [<at 0 1 RCIM_Cubic>,<at 0.15 3 RCIM_Cubic>,<at 1 1>]}}}
        CurveIndex=$System.Timeline.Time
       )
      )
     )
     2=&FloatFromCurve(
      FloatCurve=DataInterface() {propertyValues {Curve {
         defaultValue=0
         keys [
          {arriveTangent=4.1785, interpMode=RCIM_Cubic, leaveTangent=4.1785, tangentMode=RCTM_User, time=0.1535, value=0}
          {arriveTangent=1.1179, interpMode=RCIM_Cubic, leaveTangent=1.1179, tangentMode=RCTM_User, time=0.362, value=0.6228}
          {arriveTangent=0.0224, interpMode=RCIM_Cubic, leaveTangent=0.0224, tangentMode=RCTM_User, time=1, value=1}
         ]
        }}}
      CurveIndex=$Engine.System.Age
      Scale Curve=&Multiply_Float(
       A=&Multiply_Float(A=&FN_FrontEndFloatModifier(Gameplay=0.2, Front End=0.15), B=$System.TODFloat)
       B=&FloatFromCurve(
        FloatCurve=DataInterface() {propertyValues {Curve {keys [<at 0 1 RCIM_Cubic>,<at 0.15 1.5 RCIM_Cubic>,<at 1 1>]}}}
        CurveIndex=$System.Timeline.Time
       )
      )
     )
     3=?
     Index=&SpawnGroup()
     Num=::NewEnumerator1
    )
    Scale RGB=&MergeHelper_VectorSwitch(
     0=&Multiply_VectorByFloat(
      Vector=&MakeVectorFromLinearColorRGB(Color=&ColorFromCurve(
        ColorCurve=DataInterface() {propertyValues {
          AlphaCurve {keys [<at 0 1 RCIM_Cubic>,<at 1 0 RCIM_Cubic>]}
          BlueCurve {keys [<at 0 3 RCIM_Cubic>,<at 0.2962 2.3125>,<at 0.7906 2.5312>,<at 1 3 RCIM_Cubic>]}
          GreenCurve {keys [<at 0 3 RCIM_Cubic>,<at 0.2962 2.8473>,<at 0.7906 2.682>,<at 1 3 RCIM_Cubic>]}
          RedCurve {keys [<at 0 3 RCIM_Cubic>,<at 0.2962 3>,<at 0.7906 3>,<at 1 3 RCIM_Cubic>]}
         }}
        CurveIndex=$System.Timeline.Time
       ))
      Float=&Lerp_Float(
       A=&FloatFromCurve(
        FloatCurve=DataInterface() {propertyValues {Curve {keys [<at 0 1 RCIM_Cubic>, <at 0.15 2 RCIM_Cubic>, {arriveTangent=-2.8451, tangentMode=RCTM_User, time=1, value=1}]}}}
        CurveIndex=$System.Timeline.Time
       )
       B=&FloatFromCurve(
        FloatCurve=DataInterface() {propertyValues {Curve {keys [<at 0 1 RCIM_Cubic>, <at 0.15 2 RCIM_Cubic>, {arriveTangent=-2.8906, tangentMode=RCTM_User, time=1, value=1}]}}}
        CurveIndex=$System.Timeline.Time
       )
       Alpha=$User.Variant
      )
     )
     1=&Multiply_VectorByFloat(
      Vector=&MakeVectorFromLinearColorRGB(Color=&ColorFromCurve(
        ColorCurve=DataInterface() {propertyValues {
          AlphaCurve {keys [<at 0 1 RCIM_Cubic>,<at 1 0 RCIM_Cubic>]}
          BlueCurve {keys [<at 0 3 RCIM_Cubic>,<at 0.2962 2.3125>,<at 0.7906 2.5312>,<at 1 3 RCIM_Cubic>]}
          GreenCurve {keys [<at 0 3 RCIM_Cubic>,<at 0.2962 2.8473>,<at 0.7906 2.682>,<at 1 3 RCIM_Cubic>]}
          RedCurve {keys [<at 0 3 RCIM_Cubic>,<at 0.2962 3>,<at 0.7906 3>,<at 1 3 RCIM_Cubic>]}
         }}
        CurveIndex=$System.Timeline.Time
       ))
      Float=&Lerp_Float(
       A=&FloatFromCurve(
        FloatCurve=DataInterface() {propertyValues {Curve {keys [<at 0 1 RCIM_Cubic>, <at 0.15 2 RCIM_Cubic>, {arriveTangent=-2.8451, tangentMode=RCTM_User, time=1, value=1}]}}}
        CurveIndex=$System.Timeline.Time
       )
       B=&FloatFromCurve(
        FloatCurve=DataInterface() {propertyValues {Curve {keys [<at 0 1 RCIM_Cubic>, <at 0.15 2 RCIM_Cubic>, {arriveTangent=-2.8906, tangentMode=RCTM_User, time=1, value=1}]}}}
        CurveIndex=$System.Timeline.Time
       )
       Alpha=$User.Variant
      )
     )
     2=&Multiply_VectorByFloat(
      Vector=&MakeVectorFromLinearColorRGB(Color=&LerpLinearColors(StartColor=<rgba 3 3 3 1>, EndColor=<rgba 3 3 3 1>, LerpFactor=$User.Variant))
      Float=&Lerp_Float(
       A=&FloatFromCurve(
        FloatCurve=DataInterface() {propertyValues {Curve {keys [<at 0 1 RCIM_Cubic>,<at 0.15 1.5 RCIM_Cubic>,<at 1 1>]}}}
        CurveIndex=$System.Timeline.Time
       )
       B=&FloatFromCurve(
        FloatCurve=DataInterface() {propertyValues {Curve {keys [<at 0 1 RCIM_Cubic>,<at 0.15 1.5 RCIM_Cubic>,<at 1 1>]}}}
        CurveIndex=$System.Timeline.Time
       )
       Alpha=$User.Variant
      )
     )
     3=?
     Index=&SpawnGroup()
     Num=::NewEnumerator1
    )
    ScaleA=true
    ScaleRGB=true
   )
   &SocketLocation(
    SamplePosOffset=true
    Sampled Position Offset=&MergeHelper_VectorSwitch(0=<0 6.0935 -16.7724>, 1=<0 -5 0>, 2=<0 6.4728 -18.7039>, 3=?, Index=&SpawnGroup(), Num=::NewEnumerator1)
    Socket Reader=$System.Socket Reader
    Socket Sampling Mode=::NewEnumerator3
    VelocitySpeedLimit=false
   )
   &FN_Light_Attributes(
    Color=&LerpLinearColors(StartColor=<rgba 12 2.2641 1.0005 1>, EndColor=<rgba 11 7.3368 1.7156 1>, LerpFactor=$User.Variant)
    Create Color Attribute=true
    Create Exponent Attribute=true
    Create Radius Attribute=true
    Exponent=3
    Radius=30
    Scale Color=&FNTimeOfDayScalar(Night Scale=0.8, Morning Scale=0.9, Evening Scale=0.85)
   )
  ]
  rendererClasses [&NiagaraSpriteRendererProperties, &NiagaraLightRendererProperties]
  renderers [
   {rendererClass=&NiagaraSpriteRendererProperties}
   {rendererClass=&NiagaraLightRendererProperties, rendererIndex=1}
   {rendererClass=&NiagaraSpriteRendererProperties, rendererIndex=2}
  ]
 }
}
summary {
 systemName=NS_Pickaxe_ScoopPluck_Idle
 emitters [
  {emitterName=TextureGlow, rendererClasses [&NiagaraSpriteRendererProperties, &NiagaraLightRendererProperties]}
  {emitterName=SpriteGlows, rendererClasses [&NiagaraSpriteRendererProperties]}
 ]
 userVariables [
  User.bIsFrontEnd {type=NiagaraBool}
  User.bIsFrontEndPreview {type=NiagaraBool}
  User.DebugReactivity {type=NiagaraBool}
  User.KillReactivityCount {type=NiagaraInt32}
  User.Variant {type=NiagaraFloat, defaultValue=0, description=""}
 ]
}
