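# Niagara swing trail for the ScoopPluck pickaxe, built from two ribbon emitters, EdgeTrail and SoftTrail.
# Both use shared AnimTrail modules to spawn, initialize and update short-lived ribbons from weapon sockets through a system-level Socket Reader.
# User.Variant blends each trail from red toward a more translucent yellow, with time-of-day tinting and an initial brightness boost.
# Alpha curves scaled by System.AlphaFade fade both trails, and only EdgeTrail tapers in width over life.
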NiagaraSystem /BRCosmetics/Weapons/FORT_Melee/Pickaxe_ScoopPluck/FX/NS_Pickaxe_ScoopPluck_Trail
dependencies {
 usedDataInterfaces [&NiagaraDataInterfaceCurve, &NiagaraDataInterfaceSocketReader]
 usedDynamicInputs [&MergeHelper_FloatSwitch]
 usedModules [
  &self:SystemSpawnScript.SystemScriptSource.SystemScriptGraph.NiagaraNodeAssignment_1.SetVariables_#1
  &SystemState
  &self:SystemSpawnScript.SystemScriptSource.SystemScriptGraph.NiagaraNodeAssignment_0.SetVariables_#2
 ]
}
emitters {
 EdgeTrail {
  emitterUpdateScript [
   &EmitterState()
   &FN_SpawnAnimTrail_v2(AnimTrailWidthMode=::FromSecond)
  ]
  particleSpawnScript [
   /Game/Effects/Niagara/Modules/AnimTrail/FN_InitializeAnimTrail/FN_InitializeAnimTrailBoneIndecies(First Bone or Socket Index=2)
   /Game/Effects/Niagara/Modules/AnimTrail/FN_InitializeAnimTrail/FN_InitializeAnimTrail_SocketReader/FN_InitializeAnimTrail_SocketReader(
    Color:LinearColor=&LerpLinearColors(StartColor=<rgba 5 1.486 1.328 1>, EndColor=<rgba 4 3.23 1.125 0.25>, LerpFactor=$User.Variant)
    Lifetime=0.29
    Socket Reader:NiagaraDataInterfaceSocketReader=$System.Socket Reader
   )
  ]
  particleUpdateScript [
   /Game/Effects/Niagara/Modules/AnimTrail/FN_UpdateAnimTrail/FN_UpdateAnimTrail_SocketReader/FN_UpdateAnimTrail_SocketReader(Socket Reader:NiagaraDataInterfaceSocketReader=$System.Socket Reader)
   &ScaleColor(
    Scale Alpha=&FloatFromCurve(
     FloatCurve=DataInterface() {propertyValues {Curve {keys [<at 0.175 1 RCIM_Cubic>,<at 0.35 0>]}}}
     Scale Curve=&Multiply_Float(A=0.4, B=$System.AlphaFade)
    )
    Scale RGB=&Multiply_Vector(
     A=&VectorFromFloat(Value=&FloatFromCurve(FloatCurve=DataInterface() {propertyValues {Curve {keys [<at 0 1.5 RCIM_Cubic>,<at 0.13 1>]}}}))
     B=&MakeVectorFromLinearColorRGB(Color=&FNTimeOfDayColor(
       Night Color=<rgba 0.5469 0.689 1 1>
       Morning Color=<rgba 1 0.8212 0.7448 1>
       Day Color=<rgba 1 0.9304 0.8906 1>
       Evening Color=<rgba 1 0.843 0.7656 1>
      ))
    )
    ScaleA=true
    ScaleRGB=true
   )
   &ScaleRibbonWidth(Ribbon Width Scale=&FloatFromCurve(FloatCurve=DataInterface() {propertyValues {Curve {keys [<at 0 0.9>,<at 1 0>]}}}))
  ]
  rendererClasses [&NiagaraRibbonRendererProperties]
  renderers [{rendererClass=&NiagaraRibbonRendererProperties}]
 }
 SoftTrail {
  emitterUpdateScript [
   &EmitterState()
   &FN_SpawnAnimTrail_v2()
  ]
  particleSpawnScript [
   &FN_InitializeAnimTrailBoneIndecies(First Bone or Socket Index=1)
   &FN_InitializeAnimTrail_SocketReader(
    AnimTrailWidth=1.2253
    Color:LinearColor=&LerpLinearColors(StartColor=<rgba 2 0.127 0.124 1>, EndColor=<rgba 1.7 1.3899 0.4137 0.25>, LerpFactor=$User.Variant)
    Lifetime=0.26
    Socket Reader:NiagaraDataInterfaceSocketReader=$System.Socket Reader
   )
  ]
  particleUpdateScript [
   &FN_UpdateAnimTrail_SocketReader(Socket Reader:NiagaraDataInterfaceSocketReader=$System.Socket Reader)
   &ScaleColor(
    Scale Alpha=&FloatFromCurve(
     FloatCurve=DataInterface() {propertyValues {Curve {keys [<at 0.175 1 RCIM_Cubic>,<at 0.35 0>]}}}
     Scale Curve=&Multiply_Float(A=0.732, B=$System.AlphaFade)
    )
    Scale RGB=&Multiply_Vector(
     A=&VectorFromFloat(Value=&FloatFromCurve(FloatCurve=DataInterface() {propertyValues {Curve {keys [<at 0 1.5 RCIM_Cubic>,<at 0.13 1>]}}}))
     B=&MakeVectorFromLinearColorRGB(Color=&FNTimeOfDayColor(
       Night Color=<rgba 0.7188 0.8069 1 1>
       Morning Color=<rgba 1 0.8212 0.7448 1>
       Day Color=<rgba 1 0.9304 0.8906 1>
       Evening Color=<rgba 1 0.843 0.7656 1>
      ))
    )
    ScaleA=true
    ScaleRGB=true
   )
  ]
  rendererClasses [&NiagaraRibbonRendererProperties]
  renderers [{rendererClass=&NiagaraRibbonRendererProperties}]
 }
}
summary {
 systemName=NS_Pickaxe_ScoopPluck_Trail
 emitters [
  {emitterName=SoftTrail, rendererClasses [&NiagaraRibbonRendererProperties]}
  {emitterName=EdgeTrail, rendererClasses [&NiagaraRibbonRendererProperties]}
 ]
 userVariables [
  User.bIsFrontEnd {type=NiagaraBool}
  User.bIsFrontEndPreview {type=NiagaraBool}
  User.RibbonIDOffset {type=NiagaraInt32}
  User.SwingSectionIndex {type=NiagaraInt32}
  User.Variant {type=NiagaraFloat, defaultValue=0, description=""}
 ]
}
