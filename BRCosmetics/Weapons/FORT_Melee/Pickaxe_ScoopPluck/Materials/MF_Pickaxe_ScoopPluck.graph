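# Material function for the Scoop Pluck pickaxe's energy blade, adding glowing halftone dots to incoming material attributes through emissive.
# Object-stabilized screen-space UVs sample a dotted distance field, thresholded at mask-driven sizes with screen-derivative antialiasing.
# A mask input and sharpened gradient shape dot color and brightness, over optional base-color glow.
# Local-position gradients and time since TimeStart form a traveling, duration-limited reactivity pulse, with a debug looping switch.
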MaterialFunction /BRCosmetics/Weapons/FORT_Melee/Pickaxe_ScoopPluck/Materials/MF_Pickaxe_ScoopPluck
graph {nodes [
  n1 = GetMaterialAttributes(None=NamedRerouteUsage()) {attributeGetTypes [#1, #2]}
  n2 = Multiply(A=@n1.BaseColor, B=ScalarParameter() {group=Energy Blade, parameterName=BaseColor_Influence})
  n3 = TextureSample(UVs=Multiply(
    A=MaterialFunctionCall!(Position, UVRadialScale) {materialFunction=/Game/Weapons/WeaponSkins/MaterialFunction/ObjectStabilizedScreenSpaceUVs}.StabilizedUVs
    B=ScalarParameter() {defaultValue=30, group=Energy Blade, parameterName=DotTile}
   )) {samplerType=SAMPLERTYPE_LinearGrayscale, texture=/BRCosmetics/Weapons/FORT_Melee/Pickaxe_ScoopPluck/Textures/T_DottedPattern_DF}
  n4 = Subtract(A=ScalarParameter() {group=Energy Blade, parameterName=MinDotSize})
  n5 = Subtract(A=ScalarParameter() {defaultValue=1.3, group=Energy Blade, parameterName=MaxDotSize})
  n6 = NamedRerouteUsage()
  n7 = Max(A=Power(Base=ComponentMask(None=@n6) {r=true}, Exp=Constant() {r=0.5}), B=NamedRerouteUsage())
  n8 = Saturate(None=Add(
    A=NamedRerouteDeclaration(None=@n3.R) {name=Pattern DF, nodeColor=<rgba 1 0 0.3984 1>}
    B=LinearInterpolate(A=@n4, B=@n5, Alpha=@n7) {constA=-1, constB=0.3}
   ))
  n9 = Add(A=Abs(None=DDX(Value=@n8)), B=Abs(None=DDY(Value=@n8)))
  n10 = Add(A=Divide(A=Subtract(A=@n8) {constB=0.5}, B=Max(A=@n9) {constB=0})) {constB=0.5}
  n11 = ComponentMask(None=NamedRerouteUsage()) {r=true}
  n12 = Power(Base=@n11, Exp=ScalarParameter() {defaultValue=40, group=Energy Blade, parameterName=DotGradientSharpness})
  n13 = Add(
   A=Multiply(A=@n11)
   B=Multiply(A=Saturate(None=@n12), B=ScalarParameter() {defaultValue=1000, group=Energy Blade, parameterName=DotGradientBrightness})
  )
  n14 = VectorParameter() {group=Energy Blade, parameterName=Dot Color, defaultValue=<rgba 1 0.5 0.5 1>}
  n15 = NamedRerouteUsage()
  n16 = VectorParameter() {group=Reactivity, parameterName=ReactivityColor, defaultValue=<rgba 10 1.2917 0.5 1>}
  n17 = NamedRerouteUsage()
  n18 = Constant() {r=5}
  n19 = Multiply(A=OneMinus(None=Power(Base=@n17, Exp=@n18)), B=Power(Base=@n17) {constExponent=0.5})
  n20 = Multiply(
   A=Add(A=Saturate(None=@n10), B=Multiply(A=@n13, B=@n14.RGB))
   B=Multiply(A=ComponentMask(None=@n15) {r=true}, B=LinearInterpolate(B=@n16.RGB, Alpha=@n19) {constA=1})
  )
  n21 = ScalarParameter() {defaultValue=0.5, group=Energy Blade, parameterName=MinOpacity}
  FunctionOutput(None=SetMaterialAttributes!(MaterialAttributes=@n1.MaterialAttributes, BaseColor=@n1.MaterialAttributes, EmissiveColor=Add(A=@n2, B=@n20), Opacity=@n21) {attributeSetTypes [#1, #2, #3]})
  n22 = ComponentMask(None=NamedRerouteUsage()) {b=true}
  LinearInterpolate(A=@n21, Alpha=Add(A=Multiply(A=@n20, B=@n22), B=OneMinus(None=@n22)))
  NamedRerouteDeclaration(None=FunctionInput() {inputName=InMat, inputType=FunctionInput_MaterialAttributes}) {name=InMat, nodeColor=<rgba 0.7109 0 1 1>}
  NamedRerouteDeclaration(None=FunctionInput() {inputName=InMask, sortPriority=1}) {name=InMask, nodeColor=<rgba 0 1 0.3906 1>}
  n23 = LocalPosition() {localOrigin=InstancePreSkinning}
  n24 = Add(A=DotProduct(A=VertexInterpolator(VS=@n23.XYZ).PS, B=Constant4Vector() {constant=<rgba 1 0 -0.608 1>}.RGB), B=Constant() {r=28.3})
  n25 = Constant() {r=91.45}
  NamedRerouteDeclaration(None=Frac(None=Divide(A=@n24, B=@n25))) {name=BladeLengthGradient, nodeColor=<rgba 1 0.8672 0 1>}
  n26 = NamedRerouteUsage()
  n27 = DotProduct(A=VertexInterpolator(VS=LocalPosition() {localOrigin=InstancePreSkinning}.XYZ).PS, B=Constant4Vector() {constant=<rgba 0.55 0 1 1>}.RGB)
  n28 = Constant() {r=17}
  n29 = Divide(A=Abs(None=Add(A=@n27, B=@n28) {constB=16.5})) {constB=30}
  n30 = ScalarParameter() {defaultValue=-4, group=Reactivity, parameterName=ReactivitySpeed}
  n31 = Multiply(A=Subtract(A=Time(), B=ScalarParameter() {group=Reactivity, parameterName=TimeStart}), B=@n30)
  n32 = Abs(None=Multiply(A=@n30, B=ScalarParameter() {defaultValue=1, group=Reactivity, parameterName=ReactivityDuration}))
  n33 = Add(A=@n32) {constB=5}
  n34 = Add(
   A=Multiply(A=Add(A=@n26, B=@n29))
   B=StaticSwitchParameter(True=Modulo!(A=@n31, B=@n33), False=@n31) {group=Reactivity, parameterName=DebugReactivity}
  )
  n35 = Abs(None=@n34)
  n36 = Constant() {r=0.75}
  n37 = ConstantBiasScale(None=Sine(None=If(A=@n35, B=@n32, "A > B"=@n36, "A == B"=@n36, "A < B"=@n34)))
  NamedRerouteDeclaration(None=Multiply(A=Power(Base=@n37) {constExponent=1}, B=ScalarParameter() {defaultValue=1, group=Reactivity, parameterName=Reactivity On/Off})) {name=ReactivityMask, nodeColor=<rgba 0.7344 1 0 1>}
 ]}
